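# Physics asset for the Bandage Ninja pickaxe mesh, setting up physical motion for its dangling bandage bones.
# A large capsule covers the two-handed handle, and small capsules cover each bandage bone.
# The bandage chain hangs from the handle on a free-rotating joint, then forks at the fifth bone into bones six and eight.
# Twist and first-swing limits tighten along the chain, while second swing stays wide, favoring bending on one axis.
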PhysicsAsset /BRCosmetics/Weapons/FORT_Melee/Pickaxe_Bandage_Ninja/Meshes/Bandage_Ninja_Axe_Physics
bodies {
 Melee_TwoHanded_Handle [{shapeType=Capsule, length=90.7363, radius=8.1394, center=<0.464 0.5532 19.4137>, rotation=<rot 0.6106 -90.0002 -0.0598>}]
 dyn_Bandage1 [{shapeType=Capsule, length=1, radius=1, shapeName=dyn_Bandage1_capsule, rotation=<rot 0 90 90>}]
 dyn_Bandage2 [{shapeType=Capsule, length=4, radius=1, shapeName=dyn_Bandage2_capsule, center=<2 0 0>, rotation=<rot 0 90 90>}]
 dyn_Bandage3 [{shapeType=Capsule, length=4, radius=1, shapeName=dyn_Bandage3_capsule, center=<2 0 0>, rotation=<rot 0 90 90>}]
 dyn_Bandage4 [{shapeType=Capsule, length=4, radius=1, shapeName=dyn_Bandage4_capsule, center=<2 0 0>, rotation=<rot 0 90 90>}]
 dyn_Bandage5 [{shapeType=Capsule, length=4, radius=1, shapeName=dyn_Bandage5_capsule, center=<2 0 0>, rotation=<rot 0 90 90>}]
 dyn_Bandage6 [{shapeType=Capsule, length=4, radius=1, shapeName=dyn_Bandage6_capsule, center=<2 0 0>, rotation=<rot 0 90 90>}]
 dyn_Bandage8 [{shapeType=Capsule, length=4, radius=1, shapeName=dyn_Bandage8_capsule, center=<2 0 0>, rotation=<rot 0 90 90>}]
}
constraints [
 {
  bone1Name=dyn_Bandage1
  bone2Name=Melee_TwoHanded_Handle
  swing1LimitDegrees=45
  swing1Motion=Free
  swing2LimitDegrees=45
  swing2Motion=Free
  twistLimitDegrees=45
  twistMotion=Free
 }
 {bone1Name=dyn_Bandage2, bone2Name=dyn_Bandage1, swing1LimitDegrees=45, swing2LimitDegrees=45, twistLimitDegrees=45}
 {bone1Name=dyn_Bandage3, bone2Name=dyn_Bandage2, swing1LimitDegrees=15, swing2LimitDegrees=45, twistLimitDegrees=5}
 {bone1Name=dyn_Bandage4, bone2Name=dyn_Bandage3, swing1LimitDegrees=15, swing2LimitDegrees=45, twistLimitDegrees=5}
 {bone1Name=dyn_Bandage5, bone2Name=dyn_Bandage4, swing1LimitDegrees=15, swing2LimitDegrees=45, twistLimitDegrees=5}
 {bone1Name=dyn_Bandage6, bone2Name=dyn_Bandage5, swing1LimitDegrees=5, swing2LimitDegrees=45, twistLimitDegrees=5}
 {bone1Name=dyn_Bandage8, bone2Name=dyn_Bandage5, swing1LimitDegrees=5, swing2LimitDegrees=45, twistLimitDegrees=5}
]
