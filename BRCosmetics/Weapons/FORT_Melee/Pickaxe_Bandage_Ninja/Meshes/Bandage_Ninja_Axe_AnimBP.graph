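# Animation Blueprint for the Bandage Ninja pickaxe that adds secondary motion to its dangling bandage bones.
# When rigid body simulation is enabled, a RigidBody node drives the bandages, followed by small additive bone corrections.
# Otherwise, at default alphas, a world-space rotation hangs the first bandage bone downward and a Trail Controller moves the chain.
# Each update sets that rotation, the blend alpha and gravity override; the front end adds a bone pre-rotation and downward force.
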AnimBlueprint /BRCosmetics/Weapons/FORT_Melee/Pickaxe_Bandage_Ninja/Meshes/Bandage_Ninja_Axe_AnimBP
target_skeleton=&/Bandage_Ninja_Axe_Skeleton
graphs [
 AnimGraph {nodes [
   RigidBody() {
    bOverrideWorldGravity=true
    overridePhysicsAsset=&/Bandage_Ninja_Axe_Physics
    simulationSpace=BaseBoneSpace
    alphaBoolBlend {blendOption=Linear}
    baseBoneRef {boneName=Root}
    componentAppliedLinearAccClamp=<25000 25000 25000>
    componentLinearAccScale=<0.425 0.425 0.425>
    componentLinearVelScale=<0.1 0.1 0.1>
    overrideWorldGravity {z=-980}
   }
   AnimGraphNode_SubInput_0 = Input Pose()
   n1 = "Transform (Modify) Bone - Bone: dyn_Bandage1"(
    ComponentPose=Local To Component(LocalPose=@AnimGraphNode_SubInput_0)
    Alpha="To Float (Boolean)"(InBool=Get bIsInFrontEnd())
    Translation="0,0,0"
    Rotation="0,90,0"
    Scale="1,1,1"
   ) {
    rotationMode=BMM_Replace
    rotationSpace=BCS_BoneSpace
    alphaBoolBlend {blendOption=Linear}
    boneToModify {boneName=dyn_Bandage1}
    rotation {yaw=90}
   }
   n2 = "To Float (Boolean)"(InBool=Get bRigidBodyEnabled())
   n3 = RigidBody!(
    ComponentPose=@n1
    Alpha=@n2
    OverrideWorldGravity:Vector=Get Gravity Override()
    ExternalForce=Get ExternalForce() -> (ExternalForce:Vector)
   ) -> (Pose) {
    bClampLinearTranslationLimitToRefPose=true
    bOverrideWorldGravity=true
    overridePhysicsAsset=&Bandage_Ninja_Axe_Physics
    simulationSpace=BaseBoneSpace
    alphaBoolBlend {blendOption=Linear}
    baseBoneRef {boneName=Melee_TwoHanded_Handle}
    componentAppliedLinearAccClamp=<25000 25000 25000>
    componentLinearAccScale=<0.425 0.425 0.425>
    componentLinearVelScale=<0.1 0.1 0.1>
    overrideWorldGravity {z=-980}
   }
   n4 = "Transform (Modify) Bone - Bone: dyn_Bandage3"(ComponentPose=@n3, Alpha=@n2, Translation="0,0,0", Rotation="0,10,10", Scale="1,1,1") {
    rotationMode=BMM_Additive
    rotationSpace=BCS_BoneSpace
    alphaBoolBlend {blendOption=Linear}
    boneToModify {boneName=dyn_Bandage3}
    rotation {roll=10, yaw=10}
   }
   n5 = "Transform (Modify) Bone - Bone: dyn_Bandage4"(ComponentPose=@n4, Alpha=@n2, Translation="0,0,0", Rotation="0,-5,0", Scale="1,1,1") {
    rotationMode=BMM_Additive
    rotationSpace=BCS_BoneSpace
    alphaBoolBlend {blendOption=Linear}
    boneToModify {boneName=dyn_Bandage4}
    rotation {yaw=-5}
   }
   n6 = Get Alpha() -> (Alpha:float)
   n7 = "Transform (Modify) Bone - Bone: dyn_Bandage1"(ComponentPose=@n5, Alpha=@n6, Translation="0,0,0", Rotation=Get UpVector() -> (UpVector:Rotator), Scale="1,1,1") {rotationMode=BMM_Replace, rotationSpace=BCS_WorldSpace, alphaBoolBlend {blendOption=Linear}, boneToModify {boneName=dyn_Bandage1}}
   n8 = Trail controller - Bone: dyn_Bandage5(ComponentPose=@n7, Alpha=@n6) {
    chainLength=5
    rotationLimits [{limitMax=<180 180 180>}, {limitMax=<180 180 180>}, {limitMax=<180 180 180>}, {limitMax=<180 180 180>}, {limitMax=<180 180 180>}]
    rotationOffsets [<0 0 0>, <0 0 0>, <0 0 0>, <0 0 0>, <0 0 0>]
    trailBone {boneName=dyn_Bandage5}
   }
   Output Pose(Result=Component To Local(ComponentPose=@n8)) {layerGroup=None}
  ]}
 EventGraph {nodes [
   Event BlueprintUpdateAnimation()
   > n1 = Sequence!(execute) -> (then_0, then_1, then_2:Exec)
   n2 = MakeVector(Y=0, Z=-1)
   n3 = GetOwningComponent()
   n4 = Cross Product!(
    A:Vector=@n2
    B:Vector=NegateVector!(A:Vector=GetRightVector!(self:Scene Component Object Reference=@n3) -> (ReturnValue:Vector)) -> (ReturnValue:Vector)
   ) -> (ReturnValue:Vector)
   > Set UpVector(UpVector:Rotator=MakeRotationFromAxes!(Forward:Vector=@n2, Right:Vector=Cross Product(A=@n4, B=@n2), Up:Vector=@n4) -> (ReturnValue:Rotator)) -> (Output_Get:Rotator)
   > n5 = Branch(Condition=Get bRigidBodyEnabled())
   > Set Alpha(Alpha:float=Get Alpha_0() -> (Alpha_0:float)) -> (Output_Get:float)
   Set Alpha(execute=@n5.else, Alpha=Get Alpha_1() -> (Alpha_1:float))
   n6 = Gravity Override State Function DEPRECATED(
    execute=@n1.then_1
    CharacterPartAnimInstance=Self-Reference()
    GameTimeInSeconds=Get Game Time in Seconds() -> (Game Time in Seconds:double)
    GravityOverrideParams=Get Gravity Override Params()
   )
   > Set Gravity Override(Gravity Override=@n6)
   n7 = Branch(execute=@n1.then_2, Condition=Get bIsInFrontEnd())
   > Set ExternalForce(ExternalForce:Vector="0,0,-1500") -> (Output_Get:Vector)
   Set ExternalForce(execute=@n7.else, ExternalForce="0, 0, 0")
  ]}
]
variables [
 Alpha {type=number}
 Alpha_0 {type=number}
 Alpha_1 {type=number, default=1}
 ExternalForce {type=Vector}
 Game Time in Seconds {type=number}
 Gravity Override {type=Vector}
 Gravity Override Params {type=GravityOverrideParamsStruct, default {emoteOrMelee {gravityOverride=<0 0 -850>}}}
 IsRigidBodyEnabled {type=number}
 UpVector {type=Rotator}
]
